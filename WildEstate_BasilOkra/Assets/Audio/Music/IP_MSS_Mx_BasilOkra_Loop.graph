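# Stereo MetaSound source for the BasilOkra meltdown music loop, played as spatially attenuated NPC music.
# On play, a Wave Player plays the loop asset, with the StartTime input converted to time and fed to Loop Start, not playback start.
# Left and right channels are each scaled by the Multiplier input, and the player's finish fires OnFinished.
# Volume modulation is overridden to the BasilOkra control bus.
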MetaSoundSource /WildEstate_BasilOkra/Assets/Audio/Music/IP_MSS_Mx_BasilOkra_Loop
inputs [
 Metldown Loop {type=WaveAsset}
 Multiplier {type=Float}
 StartTime {type=Float}
 UE.Source.OnPlay {type=Trigger}
]
interfaces [UE.OutputFormat.Stereo:1.0, UE.Source.OneShot:1.0, UE.Source:1.1]
outputs [
 "UE.OutputFormat.Stereo.Audio:0" {type=Audio}
 "UE.OutputFormat.Stereo.Audio:1" {type=Audio}
 UE.Source.OneShot.OnFinished {type=Trigger}
]
pages [Default {
  input_defaults {
   Metldown Loop=/WildEstate_BasilOkra/Assets/Audio/Music/wav/MX_S38_IP_BasilOkra_Metltdown_Loop
   Multiplier=0.8
   StartTime=0
   UE.Source.OnPlay=null
  }
  nodes [
   WavePlayer_0 = UE.Wave Player:Stereo(Wave Asset=^Metldown Loop, Play=^UE.Source.OnPlay, Loop Start=UE.ConversionFloatToTime(In=^StartTime).Out) {also {defaults {Loop="<MetaSoundBuilderResult.SUCCEEDED: 0>"}, version [1,1]}}
   Multiply_0 = UE.Multiply:Audio by Float(AdditionalOperands=^Multiplier, PrimaryOperand=@WavePlayer_0.Out Right)
   Multiply_1 = UE.Multiply:Audio by Float(AdditionalOperands=^Multiplier, PrimaryOperand=@WavePlayer_0.Out Left)
  ]
  links [
   {dst="@UE.OutputFormat.Stereo.Audio:1", src=Multiply_0.Out}
   {dst="@UE.OutputFormat.Stereo.Audio:0", src=Multiply_1.Out}
   {dst="@UE.Source.OneShot.OnFinished", src=WavePlayer_0.On Finished}
  ]
 }]
properties {
 assetImportData=&self:AssetImportData
 attenuationSettings=&/ATT_BasilOkra_Music
 maxDistance=2100
 numChannels=2
 outputFormat=Stereo
 sampleRate=48000
 soundClassObject=&Fort_Music_NPC
 assetUserData [&self:AudioAssetUserData_0]
 modulationSettings {volumeRouting=Override, volumeModulationDestination {bEnableModulation=true, modulators [/WildEstate_BasilOkra/Assets/Audio/CB_BasilOkra]}}
}
