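# One-shot stereo MetaSound Source for the BasilOkra "Pass" music sting, defaulting to the Succeed stinger wave and routed to a music stinger sound class.
# It wraps the shared IP_MSS_Mx_BasilOkra_Sting graph, feeding it play trigger, stinger, delay time and multiplier, and passing its audio and finished trigger straight out.
# The Flourish input defaults to a puzzle-solved wave but is not wired into that inner graph.
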MetaSoundSource /WildEstate_BasilOkra/Assets/Audio/Music/IP_MSS_Mx_BasilOkra_Sting_Pass
inputs [
 DelayTime {type=Time}
 Flourish {type=WaveAsset}
 Multiplier {type=Float}
 Stinger {type=WaveAsset}
 UE.Source.OnPlay {type=Trigger}
]
interfaces [UE.OutputFormat.Stereo:1.0, UE.Source.OneShot:1.0, UE.Source:1.1]
outputs [
 "UE.OutputFormat.Stereo.Audio:0" {type=Audio}
 "UE.OutputFormat.Stereo.Audio:1" {type=Audio}
 UE.Source.OneShot.OnFinished {type=Trigger}
]
pages [Default {
  input_defaults {
   DelayTime=1.4
   Flourish=/WildEstate_BasilOkra/Assets/Audio/SFX/wav/BasilOkra_Puzzle_Solved_Flourish
   Multiplier=2
   Stinger=/WildEstate_BasilOkra/Assets/Audio/Music/wav/MX_S38_IP_BasilOkra_Succeed_STING
   UE.Source.OnPlay=null
  }
  nodes [IP_MSS_Mx_BasilOkra_Sting_0 = &/IP_MSS_Mx_BasilOkra_Sting(DelayTime=^DelayTime, Multiplier=^Multiplier, Stinger=^Stinger, UE.Source.OnPlay=^UE.Source.OnPlay)]
  links [
   {dst="@UE.OutputFormat.Stereo.Audio:0", src=IP_MSS_Mx_BasilOkra_Sting_0.UE.OutputFormat.Stereo.Audio:0}
   {dst="@UE.OutputFormat.Stereo.Audio:1", src=IP_MSS_Mx_BasilOkra_Sting_0.UE.OutputFormat.Stereo.Audio:1}
   {dst="@UE.Source.OneShot.OnFinished", src=IP_MSS_Mx_BasilOkra_Sting_0.UE.Source.OneShot.OnFinished}
  ]
 }]
properties {
 assetImportData=&self:AssetImportData
 numChannels=2
 outputFormat=Stereo
 sampleRate=48000
 soundClassObject=/Game/Sounds/FortSoundClassesAndModes/Fort_Music_Stinger_PSM
 assetUserData [&self:AudioAssetUserData_0]
}
